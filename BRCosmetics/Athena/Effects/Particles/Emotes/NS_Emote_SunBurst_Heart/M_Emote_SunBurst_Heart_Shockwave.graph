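# Unlit translucent Niagara material for the heart shockwave in the SunBurst Heart emote effect.
# Particle color drives emissive through an adjustment function.
# Opacity combines an erosion-thresholded heart texture, particle alpha and a soft band from two opposing contrast-adjusted Fresnels with depth fade.
# Quality and shading-path switches replace the Fresnel and depth-fade term with a constant one on Low, Forward and Mobile.
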Material /BRCosmetics/Athena/Effects/Particles/Emotes/NS_Emote_SunBurst_Heart/M_Emote_SunBurst_Heart_Shockwave
graph {
 properties {
  bUsedWithNanite=true
  bUsedWithNiagaraMeshParticles=true
  bUsedWithNiagaraRibbons=true
  bUsedWithNiagaraSprites=true
  blendMode=BLEND_Translucent
  pixelDepthOffsetMode=PDOM_Legacy
  shadingModel=MSM_Unlit
  usedShadingModels=Unlit
 }
 nodes [
  n1 = TextureCoordinate()
  n2 = Constant() {r=0.5}
  Add(A=Multiply(A=Subtract(A=@n1, B=@n2)), B=@n2)
  n3 = ParticleColor()
  MaterialFunctionCall(Emissive=@n3.RGB) {materialFunction=&MF_FX_EmissiveAdjustments}
  n4 = TextureSample() {samplerType=SAMPLERTYPE_LinearGrayscale, texture=&/T_SunBurst_Heart}
  n5 = Divide(A=Subtract(A=ChannelMaskParameter(None=@n4.RGB) {parameterName=Opacity Channel}.R, B=ScalarParameter() {defaultValue=0.6, parameterName=Erode}) {constB=0.8}) {constB=0.5}
  n6 = ScalarParameter() {defaultValue=1, parameterName=FresnelExponent}
  n7 = Fresnel(ExponentIn=ScalarParameter() {defaultValue=1, parameterName=FresnelExponent2}) {exponent=1}
  n8 = Multiply(
   A=MaterialFunctionCall(In=Fresnel(ExponentIn=@n6) {baseReflectFraction=0.1, exponent=1}, Contrast=Constant() {r=0.1}) {materialFunction=&CheapContrast}.Result
   B=MaterialFunctionCall(In=OneMinus(None=@n7), Contrast=Constant() {r=1}) {materialFunction=&CheapContrast}.Result
  )
  n9 = ScalarParameter() {defaultValue=1, parameterName=FresnelBrightness}
  n10 = DepthFade(Opacity=Saturate(None=Multiply(A=@n8, B=@n9)), FadeDistance=ScalarParameter() {parameterName=DepthFade_Distance})
  n11 = Constant() {r=1}
  n12 = QualitySwitch(Default=@n10, Low=@n11, High=@n10, Medium=@n10, Epic=@n10)
  Multiply(A=Multiply(A=Saturate(None=@n5), B=@n3.A), B=ShadingPathSwitch(Default=@n12, Deferred=@n12, Forward=@n11, Mobile=@n11))
  Saturate()
 ]
}
