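# Glossy dark-red lit material for Niagara mesh particles in the SunBurst Heart emote, apparently a visualizer element.
# A Fresnel-warped gradient texture adds a thin-film tint, with its mask also weighted by inverse roughness.
# World position offset, gated by dynamic parameter UseWPO, uses vertex red and the X-to-Y scale ratio to counter stretching along local X.
# A local-X distance mask is derived for blending surface properties.
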Material /BRCosmetics/Athena/Effects/Particles/Emotes/NS_Emote_SunBurst_Heart/M_Emote_SunBurst_Heart_Viz
graph {
 properties {bUsedWithNiagaraMeshParticles=true, pixelDepthOffsetMode=PDOM_Legacy, usedShadingModels=Default Lit}
 nodes [
  Constant3Vector() {constant=<rgba 1 1 1 1>}
  Constant() {desc=Specular, r=0.2086}
  Constant() {desc=Metallic}
  n1 = OneMinus(None=NamedRerouteUsage())
  n2 = ScalarParameter() {defaultValue=1, group=ThinFilm, parameterName=ThinFilm_RoughnessScale}
  n3 = Multiply(
   A=Clamp(None=Multiply(A=@n1, B=@n2))
   B=Fresnel(ExponentIn=ScalarParameter() {defaultValue=1.2164, group=ThinFilm, parameterName=ThinFilm_Exponent}) {baseReflectFraction=0}
  )
  LinearInterpolate(B=@n3, Alpha=NamedRerouteUsage())
  n4 = MaterialFunctionCall() {materialFunction=&LocalPosition}
  n5 = PreSkinnedLocalBounds()
  n6 = MaterialFunctionCall(Float3=Divide(A=@n4."Instance Local Position (Excluding Offsets)", B=@n5.Half Extents)) {materialFunction=&BreakOutFloat3Components}
  n7 = Abs(None=@n6.R)
  n8 = VertexColor()
  n9 = Constant()
  n10 = MaterialFunctionCall(X=Multiply(A=OneMinus(None=@n7), B=LinearInterpolate(Alpha=@n8.R) {constA=-1}), Y=@n9, Z=@n9) {materialFunction=&MakeFloat3}
  n11 = PreSkinnedLocalBounds()
  n12 = MaterialFunctionCall() {materialFunction=&ObjectScale}
  n13 = Divide(A=@n12.Scale X, B=Max(A=@n12.Scale Y, B=@n12.Scale Y))
  n14 = Constant() {r=0.666}
  n15 = If(A=@n13, B=@n14, "A > B"=@n13, "A == B"=@n14, "A < B"=@n14) {constB=-0.25}
  n16 = Multiply(A=ComponentMask(None=Divide(A=@n11.Half Extents, B=@n15)) {r=true}, B=Subtract(A=@n15))
  LinearInterpolate(
   B=Transform(None=Multiply(A=@n10.Result, B=@n16)) {transformSourceType=TRANSFORMSOURCE_Instance}
   Alpha=DynamicParameter() {defaultValue=<rgba 0 0 0 0>, paramNames [UseWPO,Param2,Param3,Param4]}.UseWPO
  )
  LinearInterpolate(
   A=Transform(None=Constant3Vector() {constant=<rgba 1 0 0 0>}) {transformSourceType=TRANSFORMSOURCE_Instance}
   B=PixelNormalWS()
   Alpha=NamedRerouteUsage()
  )
  n17 = Constant3Vector() {constant=<rgba 0.2384 0.0116 0.011 1>}
  n18 = ScalarParameter() {defaultValue=0.5, group=ThinFilm, parameterName=ThinFilm_Warp}
  n19 = TextureSampleParameter2D(UVs=Multiply(A=ScalarParameter() {defaultValue=1, group=ThinFilm, parameterName=ThinFilm_Scale}, B=Fresnel(ExponentIn=@n18) {baseReflectFraction=0})) {
   group=ThinFilm
   parameterName=Thin Film Texture
   texture=/Game/Characters/Player/Female/Medium/Bodies/F_MED_Witch/Skins/Dark/Textures/Dark_Witch_Gradient_Purple
  }
  n20 = Multiply(A=@n19.RGB, B=ScalarParameter() {defaultValue=0.5, group=ThinFilm, parameterName=ThinFilm_Intensity})
  NamedRerouteDeclaration(None=LinearInterpolate(A=@n17, B=Add(A=@n17, B=@n20), Alpha=@n3)) {name=Base Color, nodeColor=<rgba 1 0 0.7734 1>}
  NamedRerouteDeclaration(None=Constant() {desc=Roughness, r=0.0802}) {name=Roughness, nodeColor=<rgba 0.9922 1 0 1>}
  n21 = Multiply(A=ComponentMask(None=Divide(A=@n4.Instance Local Position, B=@n5.Half Extents)) {r=true}) {constB=5}
  NamedRerouteDeclaration(None=Saturate(None=Power(Base=@n21) {constExponent=3})) {name=Distance Mask, nodeColor=<rgba 0.6641 0 1 1>}
  NamedRerouteUsage()
  Saturate(None=@n6.R)
 ]
}
