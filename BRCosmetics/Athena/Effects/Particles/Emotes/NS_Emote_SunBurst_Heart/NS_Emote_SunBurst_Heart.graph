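# Niagara system for the SunBurst Heart emote, an audio-reactive pink-purple burst sampled from a user-supplied static mesh.
# FN_NPC_EventAudio feeds high, mid and low levels into spawn rates, sizes and lights.
# Audio_Viz_Cubes_GPU stretches surface-oriented meshes with high and low levels.
# SparklesHand combines mid-driven spawns with bursts, using spawn groups as visibility tags for sprite, mesh and light renderers; Empty001 is disabled.
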NiagaraSystem /BRCosmetics/Athena/Effects/Particles/Emotes/NS_Emote_SunBurst_Heart/NS_Emote_SunBurst_Heart
dependencies {
 usedDataInterfaces [&NiagaraDataInterfaceArrayInt32, &NiagaraDataInterfaceArrayFloat, &NiagaraDataInterfaceCurve]
 usedDynamicInputs [
  &Multiply_Int
  &MakeIntFromFloat
  &ReturnNormalizedExecIndex
  &Multiply_FloatByInt
  &ModuloFloat
  &Add_Float
 ]
 usedModules [
  &self:SystemSpawnScript.SystemScriptSource.SystemScriptGraph.NiagaraNodeAssignment_9.SetVariables_#1
  /Game/Effects/Niagara/Modules/Audio/FN_NPC_EventAudio
  &SystemState
 ]
}
emitters {
 Audio_Viz_Cubes_GPU {
  emitterUpdateScript [
   &EmitterState()
   &SpawnRate(
    SpawnRate=&Multiply_Float(A=60, B=&Power(a=$System.FN_NPC_EventAudio.Audio High, b=2))
    Use Spawn Probability=false
   ) {module=SpawnRate001}
   &SpawnRate(
    SpawnGroup=1
    SpawnRate=&Multiply_Float(A=30, B=&Power(a=$System.FN_NPC_EventAudio.Audio Low, b=2))
    Use Spawn Probability=false
   ) {module=SpawnRate003}
  ]
  particleSpawnScript [
   &InitializeParticle(
    Lifetime=&SelectFloatFromArray(Direct Array Index=&SpawnGroup(), Float Selection Array=DataInterface() {propertyValues {FloatData [1,1]}})
    Mesh Scale Mode=::NewEnumerator3
    Mesh Uniform Scale=&SelectFloatFromArray(Direct Array Index=&SpawnGroup(), Float Selection Array=DataInterface() {propertyValues {FloatData [0.2,0.4]}})
    UsePositionOffset=false
    Write Material Random=true
   )
   /Niagara/Modules/Spawn/Location/StaticMeshLocation(
    Interpolate Mesh Sampling=false
    MaskSpawnGroup=false
    OffsetAlongNormal=false
    Orientation Sampling:ENiagara_AttributeSamplingApplyOutput=::NewEnumerator1
    OverrideRandomSeed=false
    SamplePosOffset=false
    Static Mesh:NiagaraDataInterfaceStaticMesh=$User.StaticMesh
   )
   /SetVariables(Particles.SurfaceNormal:Vector3f=&Lerp_Vector(
     A=&NormalizeVector(
      Vector To Normalize=$Particles.StaticMeshLocation.SampledPosition
      Safe Normalize=true
      Safe Fallback Threshold=0
      Safe Fallback Vector=<0 0 0.001>
     )
     B=$Particles.StaticMeshLocation.SampledNormal
     Alpha=0.5
    )) {isSetParameters=true}
   &InitialMeshOrientation(Mesh Orientation Mode=::NewEnumerator2, Orientation Vector=$Particles.SurfaceNormal, Use Rotation Vector=false)
  ]
  particleUpdateScript [
   &ParticleState()
   &ScaleMeshSize(Scale Factor=&Lerp_Vector(
     A=<0.1 1 1>
     B=<1 1 1>
     Alpha=&FloatFromCurve(FloatCurve=DataInterface() {propertyValues {Curve {keys [
          {interpMode=RCIM_Cubic, tangentMode=RCTM_User, time=0, value=0}
          {interpMode=RCIM_Cubic, tangentMode=RCTM_User, time=0.2, value=1}
          {interpMode=RCIM_Cubic, tangentMode=RCTM_User, time=0.8, value=1}
          {tangentMode=RCTM_User, time=1, value=0}
         ]}}})
    )) {module=ScaleMeshSize003}
   &ScaleMeshSize(Scale Factor=&MakeVector(
     X=&self:Dynamic_Float_Array(
      Index:NiagaraInt32=&SpawnGroup()
      Direct Array Mode:ENiagara_WrapClamp=::NewEnumerator0
      "Index [0]":NiagaraFloat=&Lerp_Float(A=1, B=10, Alpha=$System.FN_NPC_EventAudio.Audio High)
      "Index [1]":NiagaraFloat=&Lerp_Float(A=0.5, B=3, Alpha=$System.FN_NPC_EventAudio.Audio Low)
      "Index [2]":NiagaraFloat=0
      "Index [3]":NiagaraFloat=0
     )
     Y=1
     Z=1
    )) {module=ScaleMeshSize004}
   /SetVariables(Particles.Position=&AddVectorToPosition(
     Position=$Particles.Initial.Position
     Vector=&Multiply_VectorByFloat(Vector=$Particles.SurfaceNormal, Float=-1)
    )) {isSetParameters=true}
  ]
  rendererClasses [&NiagaraMeshRendererProperties]
  renderers [{rendererClass=&NiagaraMeshRendererProperties, rendererIndex=0}]
 }
 Empty001 {
  bEnabled=false
  emitterUpdateScript [
   &EmitterState()
   &SpawnBurst_Instantaneous(Use Spawn Probability=false, UseLoopCountLimit=true)
  ]
  particleSpawnScript [&InitializeParticle(Color Mode=::NewEnumerator0, Position Offset=<0 0 200>, UsePositionOffset=true, Write Material Random=true)]
  particleUpdateScript [
   &Color(Color=$System.FN_NPC_EventAudio.Audio Remapped)
   &Color(Color=/Niagara/DynamicInputs/LinearColor/MakeLinearColorFromFloat4(Vector:Vector4f=&MakeVector4(X=0, Y=0, Z=0, W=0))) {enabled=false, module=Color001}
   &ParticleState(Kill Particles When Lifetime Has Elapsed=false)
  ]
  rendererClasses [&NiagaraSpriteRendererProperties]
  renderers [{rendererClass=&NiagaraSpriteRendererProperties}]
 }
 SparklesHand {
  emitterUpdateScript [
   &EmitterState()
   &SpawnRate(
    SpawnRate=&Multiply_Float(A=30, B=&Power(a=$System.FN_NPC_EventAudio.Audio Mid, b=2))
    Use Spawn Probability=false
   ) {module=SpawnRate001}
   &SpawnBurst_Instantaneous(Spawn Count=9, Spawn Group=1, Use Spawn Probability=false, UseLoopCountLimit=false) {module=SpawnBurst_Instantaneous001}
   &SpawnBurst_Instantaneous(Spawn Group=2, Use Spawn Probability=false, UseLoopCountLimit=false)
  ]
  particleSpawnScript [
   /SetVariables(Particles.VisibilityTag=&SelectIntFromArray(Direct Array Index=&SpawnGroup(), Int Selection Array=DataInterface() {propertyValues {IntData [0,1,2]}})) {isSetParameters=true}
   &StaticMeshLocation(
    MaskSpawnGroup=false
    Mesh Sampling Type:ENiagara_StaticSamplingMode=::NewEnumerator2
    OffsetAlongNormal=false
    Orientation Sampling:ENiagara_AttributeSamplingApplyOutput=::NewEnumerator1
    OverrideRandomSeed=false
    SamplePosOffset=false
    Static Mesh:NiagaraDataInterfaceStaticMesh=$User.StaticMesh
    Write To Particle Position=false
   )
   &InitializeParticle(
    Color=<rgba 0.75 0 1 1>
    Lifetime=&Multiply_Float(
     A=&RandomRangeFloat(Minimum=0.9, Maximum=1)
     B=&SelectFloatFromArray(
      Direct Array Index=&SpawnGroup()
      Float Selection Array=DataInterface() {propertyValues {FloatData [0.333,0.8,0.4]}}
     )
    )
    Mesh Scale Mode=::NewEnumerator1
    Mesh Uniform Scale Max=1.05
    Position Offset=&MakeCustomVectorFromBool(
     Bool=&SetBoolByIntComparison(A=$Emitter.SpawnGroup, B=2, Comparison Type=::NewEnumerator2)
     True Vector=$System.LocationOffset
     False Vector=&Add_Vector(A=$Particles.StaticMeshLocation.SampledPosition, B=$System.LocationOffset)
    )
    Sprite Rotation Angle=&RandomRangeFloat(Minimum=-15, Maximum=15)
    Sprite Rotation Mode=::NewEnumerator0
    Sprite Size Mode=::NewEnumerator3
    Uniform Sprite Size=&self:Dynamic_Float_Array(
     Index:NiagaraInt32=&SpawnGroup()
     Direct Array Mode:ENiagara_WrapClamp=::NewEnumerator0
     "Index [0]":NiagaraFloat=&RandomRangeFloat(Minimum=80, Maximum=100)
     "Index [1]":NiagaraFloat=&RandomRangeFloat(Minimum=10, Maximum=15)
     "Index [2]":NiagaraFloat=1
     "Index [3]":NiagaraFloat=1
    )
    UsePositionOffset=true
    Write Material Random=true
   )
   &AddVelocity(
    Velocity=&NormalizeVector(
     Vector To Normalize=$Particles.StaticMeshLocation.SampledPosition
     Safe Normalize=true
     Safe Fallback Threshold=0
     Safe Fallback Vector=<0 0 0.001>
    )
    Velocity Speed Scale=&Multiply_Float(
     A=20
     B=&SelectFloatFromArray(
      Direct Array Index=&SpawnGroup()
      Float Selection Array=DataInterface() {propertyValues {FloatData [0.001,4,0]}}
     )
    )
   )
   &SpriteFacingAndAlignment(
    Sprite Facing=&NormalizeVector(
     Vector To Normalize=$Particles.StaticMeshLocation.SampledPosition
     Safe Normalize=true
     Safe Fallback Threshold=0
     Safe Fallback Vector=<0 0 0.001>
    )
    WriteAlignment=false
    WriteFacing=true
   )
   &SubUVAnimation(End Frame=3, SubUV Animation Mode:ENiagara_SubUVLookupMode=::NewEnumerator2)
  ]
  particleUpdateScript [
   &ParticleState()
   &Drag(Drag=&RandomRangeFloat(Minimum=4, Maximum=5), Use Linear Drag=true, Use Rotational Drag=false)
   &SolveForcesAndVelocity(Clamp Velocity=false, Limit Acceleration=false)
   &ScaleColor(
    Scale Alpha=&FloatFromCurve(
     FloatCurve=DataInterface() {propertyValues {Curve {keys [{interpMode=RCIM_Cubic, tangentMode=RCTM_User, time=0.5, value=1}, {interpMode=RCIM_Cubic, tangentMode=RCTM_User, time=1, value=0}]}}}
     Scale Curve=0.75
    )
    Scale RGB=&Lerp_Vector(
     A=<1 1 1>
     B=<3 3 3>
     Alpha=&FloatFromCurve(FloatCurve=DataInterface() {propertyValues {Curve {keys [{arriveTangent=-1.962, interpMode=RCIM_Cubic, time=0, value=1}, <at 1 0 RCIM_Cubic>]}}})
    )
    ScaleA=true
    ScaleRGB=true
   )
   &ScaleSpriteSize(
    Scale Factor=&Multiply_Vector2D_ByFloat(
     Float=&self:Dynamic_Float_Array(
      Index:NiagaraInt32=&SpawnGroup()
      Direct Array Mode:ENiagara_WrapClamp=::NewEnumerator0
      "Index [0]":NiagaraFloat=&Lerp_Float(A=0.1, B=1, Alpha=$System.FN_NPC_EventAudio.Audio Mid)
      "Index [1]":NiagaraFloat=&FloatFromCurve(FloatCurve=DataInterface() {propertyValues {Curve {keys [
           {arriveTangent=2.3721, interpMode=RCIM_Cubic, leaveTangent=2.3721, tangentMode=RCTM_User, time=0, value=0.5}
           {interpMode=RCIM_Cubic, leaveTangent=0.8037, time=0.5, value=1}
          ]}}})
      "Index [2]":NiagaraFloat=1
      "Index [3]":NiagaraFloat=1
     )
     Vector2D=&MakeVector2D(
      X=&self:Dynamic_Float_Array(
       Index:NiagaraInt32=&SpawnGroup()
       Direct Array Mode:ENiagara_WrapClamp=::NewEnumerator0
       "Index [0]":NiagaraFloat=&FloatFromCurve(FloatCurve=DataInterface() {propertyValues {Curve {keys [<at 0 0.2 RCIM_Cubic>, {arriveTangent=0.8037, interpMode=RCIM_Cubic, leaveTangent=0.8037, tangentMode=RCTM_User, time=1, value=1}]}}})
       "Index [1]":NiagaraFloat=1
       "Index [2]":NiagaraFloat=1
       "Index [3]":NiagaraFloat=1
      )
      Y=1
     )
    )
    Scale Sprite Size Mode=::NewEnumerator2
   )
   &CameraOffset(Camera Offset Amount=2.5)
   &ScaleMeshSize(Scale Factor=&Lerp_Vector(
     A=<1 1 1>
     B=<1.3 1.6 1.3>
     Alpha=&FloatFromCurve(FloatCurve=DataInterface() {propertyValues {Curve {keys [
          {arriveTangent=4.1785, interpMode=RCIM_Cubic, leaveTangent=4.1785, tangentMode=RCTM_User, time=0, value=0}
          {arriveTangent=1, interpMode=RCIM_Cubic, leaveTangent=1, time=0.25, value=0.6228}
          {arriveTangent=0.0224, interpMode=RCIM_Cubic, leaveTangent=0.0224, tangentMode=RCTM_User, time=1, value=1}
         ]}}})
    ))
   /SetVariables(Particles.Color_Hearts:LinearColor=&MakeLinearColorFromVectorAndFloat(
     "Vector (RGB)"=&Lerp_Vector(
      A=<0.75 0 1>
      B=<0.25 0 0>
      Alpha=&FloatFromCurve(FloatCurve=DataInterface() {propertyValues {Curve {keys [{arriveTangent=3.5387, interpMode=RCIM_Cubic, time=0, value=0}, {interpMode=RCIM_Cubic, tangentMode=RCTM_User, time=0.5, value=1}]}}})
     )
     "Float (Alpha)"=&FloatFromCurve(FloatCurve=DataInterface() {propertyValues {Curve {keys [{interpMode=RCIM_Cubic, tangentMode=RCTM_User, time=0, value=1}, {interpMode=RCIM_Cubic, tangentMode=RCTM_User, time=1, value=0}]}}})
    )) {isSetParameters=true}
   &DynamicMaterialParameters(
    Index 0 Param 1=&FloatFromCurve(FloatCurve=DataInterface() {propertyValues {Curve {keys [{interpMode=RCIM_Cubic, tangentMode=RCTM_User, time=0.15, value=0}, {interpMode=RCIM_Cubic, tangentMode=RCTM_User, time=0.4, value=0.85}]}}})
    Index 0 Param 2=0.05
    Module.Param0WriteEnabled=true
    Module.Param1WriteEnabled=true
    Module.Param2WriteEnabled=true
    Module.Param3WriteEnabled=true
   )
   /SetVariables(Particles.Color_Mesh:LinearColor=&LerpLinearColors(
     StartColor=<rgba 0.5583 0.0578 0.7758 2>
     EndColor=<rgba 0 0 0 0>
     LerpFactor=&FloatFromCurve(FloatCurve=DataInterface() {propertyValues {Curve {keys [
          {arriveTangent=2.3984, interpMode=RCIM_Cubic, time=0, value=0}
          {arriveTangent=1, interpMode=RCIM_Cubic, leaveTangent=1, time=0.25, value=0.66}
          <at 1 1 RCIM_Cubic>
         ]}}})
    )) {isSetParameters=true}
   /SetVariables() {isSetParameters=true}
   &FN_Light_Attributes(
    Color=<rgba 0.5526 0.0388 0.7758 5>
    Create Color Attribute=true
    Create Radius Attribute=true
    Radius=&Multiply_Float(
     A=200
     B=&Lerp_Float(
      A=0.5
      B=1
      Alpha=&Max_Float(
       A:NiagaraFloat=$System.FN_NPC_EventAudio.Audio Mid
       B:NiagaraFloat=&FloatFromCurve(
        FloatCurve=DataInterface() {propertyValues {Curve {keys [{interpMode=RCIM_Cubic, tangentMode=RCTM_User, time=0, value=1}, {interpMode=RCIM_Cubic, tangentMode=RCTM_User, time=1, value=0}]}}}
        Scale Curve=0.8
       )
      )
     )
    )
    Scale Color=&Max_Float(
     A:NiagaraFloat=$System.FN_NPC_EventAudio.Audio Mid
     B:NiagaraFloat=&FloatFromCurve(
      FloatCurve=DataInterface() {propertyValues {Curve {keys [{interpMode=RCIM_Cubic, tangentMode=RCTM_User, time=0, value=1}, {interpMode=RCIM_Cubic, tangentMode=RCTM_User, time=1, value=0.1}]}}}
      Scale Curve=0.8
     )
    )
   )
   &FN_Light_Attributes(
    Color=<rgba 0.5526 0.0388 0.7758 4>
    Create Color Attribute=true
    Create Radius Attribute=true
    Radius=&Multiply_Float(
     A=/Niagara/DynamicInputs/Math/CalculateParticleRadius(
      Radius Calculation Type:ENiagaraCollisionRadiusOptions=::NewEnumerator0
      Particle Radius:NiagaraFloat=1
      Mesh Dimensions:Vector3f=<1 1 1>
      Method for Calculating Particle Radius:ENiagaraCalculateRadiusOptions=::NewEnumerator2
     )
     B=1.25
    )
    Scale Color=$System.FN_NPC_EventAudio.Audio Mid
   ) {module=FN_Light_Attributes001}
   /SetVariables(Particles.Position Light Shafts:NiagaraPosition=&AddVectorToPosition(
     Position=$Particles.Initial.Position
     Vector=&Multiply_VectorByFloat(
      Vector=$Particles.StaticMeshLocation.SampledNormal
      Float=&Multiply_Float(
       A=&CalculateParticleRadius(
        Radius Calculation Type:ENiagaraCollisionRadiusOptions=::NewEnumerator0
        Particle Radius:NiagaraFloat=1
        Mesh Dimensions:Vector3f=<1 1 1>
        Method for Calculating Particle Radius:ENiagaraCalculateRadiusOptions=::NewEnumerator2
       )
       B=0.6
      )
     )
    )) {isSetParameters=true}
  ]
  rendererClasses [
   &NiagaraSpriteRendererProperties
   &NiagaraLightRendererProperties
   &NiagaraMeshRendererProperties
  ]
  renderers [
   {rendererClass=&NiagaraSpriteRendererProperties}
   {rendererClass=&NiagaraSpriteRendererProperties, rendererIndex=1}
   {rendererClass=&NiagaraLightRendererProperties, rendererIndex=2}
   {rendererClass=&NiagaraMeshRendererProperties, rendererIndex=3}
   {rendererClass=&NiagaraLightRendererProperties, rendererIndex=4}
  ]
 }
}
summary {
 systemName=NS_Emote_SunBurst_Heart
 emitters [
  {
   emitterName=SparklesHand
   rendererClasses [
    &NiagaraSpriteRendererProperties
    &NiagaraLightRendererProperties
    &NiagaraMeshRendererProperties
   ]
  }
  {emitterName=Audio_Viz_Cubes_GPU, rendererClasses [&NiagaraMeshRendererProperties]}
  {bEnabled=false, emitterName=Empty001, rendererClasses [&NiagaraSpriteRendererProperties]}
 ]
 userVariables [
  User.bIsFrontEnd {type=NiagaraBool}
  User.bIsFrontEndPreview {type=NiagaraBool}
  User.LocationOffset {type=Vector3f, description="", defaultValue {struct=Vector3f, value=<0 0 0>}}
  User.StaticMesh {
   type=NiagaraDataInterfaceStaticMesh
   description=""
   defaultValue {
    struct=NiagaraExt_VariableValue_DataInterface
    value {
     dataInterface=&self:NiagaraDataInterfaceStaticMesh_0
     dataInterfaceClass=&NiagaraDataInterfaceStaticMesh
    }
   }
  }
 ]
}
